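# CornStalks_Enter_Cue is a foley SoundCue for entering corn stalks, stored with the Cornfield gameplay cue sounds.
# A single Wave Player feeds the output with one wave, CornStalks_Enter_01, at reduced volume and with no randomization.
# It uses the Fort_Consumables sound class and borrows the third-person walk footstep attenuation.
# Overridden concurrency limits instances per owner to a small count, stopping the oldest.
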SoundCue /FNGameplayCues/Sounds/Fort_Foley_Sounds/Cornfield/CornStalks_Enter_Cue
root=WavePlayer_1
nodes [{id=WavePlayer_1, properties {soundWaveAssetPtr=&/CornStalks_Enter_01}}]
properties {
 attenuationSettings=/Game/Sounds/Attenuation/Fort_Player_Footsteps_3P_Walk_Athena_Att
 bOverrideConcurrency=true
 maxDistance=3584
 soundClassObject=&Fort_Consumables
 volumeMultiplier=0.5
 concurrencyOverrides {bLimitToOwner=true, maxCount=2, resolutionRule=StopOldest}
}
